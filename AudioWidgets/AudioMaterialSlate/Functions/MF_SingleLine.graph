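# Material function for AudioWidgets slate materials that builds masks for a single straight line at LinePosition across UV space.
# Nested If nodes cut a hard band using LineThickness scaled by 0.01.
# SmoothSteps add a SplitFill output filling one side of the line and a Glow output fading out from both edges.
# Both branches of the Use V switch read the G component, so the switch has no effect as wired.
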MaterialFunction /AudioWidgets/AudioMaterialSlate/Functions/MF_SingleLine
graph {
 properties {bExposeToLibrary=true}
 nodes [
  n1 = MaterialFunctionCall(Float2=FunctionInput(Preview=TextureCoordinate()) {bUsePreviewValueAsDefault=true, inputName=UV, inputType=FunctionInput_Vector2}) {materialFunction=&BreakOutFloat2Components}
  n2 = StaticSwitch(
   True=@n1.G
   False=@n1.G
   Value=FunctionInput(Preview=StaticBoolParameter() {parameterName=Use V}) {bUsePreviewValueAsDefault=true, inputName=Use V, inputType=FunctionInput_StaticBool}
  )
  n3 = FunctionInput(Preview=Constant() {r=0.5}) {bUsePreviewValueAsDefault=true, inputName=LinePosition, inputType=FunctionInput_Scalar}
  n4 = Multiply(A=FunctionInput(Preview=Constant() {r=9.7499}) {bUsePreviewValueAsDefault=true, inputName=LineThickness, inputType=FunctionInput_Scalar}) {constB=0.01}
  n5 = Constant() {r=1}
  n6 = If(A=@n2, B=Subtract(A=@n3, B=@n4), "A > B"=@n5, "A == B"=@n5, "A < B"=Constant())
  FunctionOutput(None=If(A=@n2, B=Add(A=@n3, B=@n4), "A > B"=Constant(), "A == B"=@n6, "A < B"=@n6))
  n7 = FunctionInput(Preview=Constant() {r=0.02}) {bUsePreviewValueAsDefault=true, inputName=LineSmoothStepThickness, inputType=FunctionInput_Scalar}
  n8 = Add(A=@n3, B=@n4)
  FunctionOutput(None=SmoothStep(Min=Subtract(A=@n3, B=@n7), Max=@n8, Value=@n2)) {
   description="Creates a smooth mask (smoothness controlled by linesmoothstep thickness) and fills half of the space instead of just a line. Useful for filling under a line (and simply invert it to fill the other side instead)"
   outputName=SplitFill
  }
  n9 = FunctionInput(Preview=Constant() {r=0.2}) {bUsePreviewValueAsDefault=true, inputName=LineSmoothGlowThickness, inputType=FunctionInput_Scalar}
  n10 = Add(A=@n8, B=@n9)
  n11 = Subtract(A=@n3, B=@n4)
  n12 = Subtract(A=@n11, B=@n9)
  FunctionOutput(None=Multiply(A=SmoothStep(Min=@n10, Max=@n8, Value=@n2), B=SmoothStep(Min=@n12, Max=@n11, Value=@n2))) {outputName=Glow}
 ]
}
