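# Material function for audio widget UI that draws a 2:1 rounded box with a glowing center line, outputting Color and Mask.
# Two GeneratedRoundRect shapes on center-scaled UVs form the outer mask and inset background, leaving an outline between them.
# MF_SingleLine supplies the line and glow, scaled by the CenterlineGlowMult input.
# Subtracting the border ring from the line keeps it off the outline when blended over the box colors.
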MaterialFunction /AudioWidgets/AudioMaterialSlate/Functions/MF_21Box_Filter
graph {
 properties {bExposeToLibrary=true}
 nodes [
  n1 = VectorParameter() {parameterName=BoxOutline, defaultValue=<rgba 0.1562 0.1562 0.1562 1>}
  n2 = VectorParameter() {parameterName=BoxBG, defaultValue=<rgba 0.0234 0.0234 0.0234 1>}
  n3 = TextureCoordinate()
  n4 = MaterialFunctionCall(UVs=@n3, Texture Scale=Constant2Vector() {g=2, r=1}) {materialFunction=&ScaleUVsByCenter}
  n5 = ScalarParameter() {defaultValue=0.9609, parameterName=BoxSize}
  n6 = AppendVector(A=@n5, B=Multiply(A=@n5) {constB=0.5})
  n7 = MaterialFunctionCall!(
   UV Coords=@n4.UVs
   Box Dimensions=Subtract(A=@n6, B=ScalarParameter() {defaultValue=0.018, parameterName=BoxBorderSize})
   CornerRadius=ScalarParameter() {defaultValue=0.01, parameterName=CornerRadius}
   Sharpness=ScalarParameter() {defaultValue=56.137, parameterName=BoxSharpness}
   Center
  ) {materialFunction=/Engine/Functions/Engine_MaterialFunctions01/Gradient/GeneratedRoundRect}
  n8 = Constant() {r=0.02}
  n9 = MaterialFunctionCall!(
   LinePosition
   LineThickness=ScalarParameter() {defaultValue=0.2, parameterName=CenterLineThickness}
   Use V=StaticBool() {value=true}
   LineSmoothStepThickness=ScalarParameter() {defaultValue=0.1561, parameterName=CenterLineGlowSize}
   LineSmoothGlowThickness
   UV=Multiply(
    A=@n3
    B=MaterialFunctionCall!(Crop Right=@n8, Crop Bottom=@n8, Crop Top=@n8, Crop Left=@n8) {materialFunction=/Engine/Functions/Engine_MaterialFunctions01/Texturing/UVCropping}.Result
   )
  ) {materialFunction=&/MF_SingleLine}
  n10 = ScalarParameter() {defaultValue=0.3485, parameterName=CenterLineGlow}
  n11 = FunctionInput(Preview=Constant() {r=1}) {bUsePreviewValueAsDefault=true, inputName=CenterlineGlowMult, inputType=FunctionInput_Scalar}
  n12 = Add(A=Multiply(A=@n9.SplitFill, B=Multiply(A=@n10, B=@n11)), B=@n9.Result)
  n13 = VectorParameter() {parameterName=LineColor, defaultValue=<rgba 0.0219 0.1531 0.1562 1>}
  n14 = MaterialFunctionCall(UV Coords=@n4.UVs, Box Dimensions=@n6, CornerRadius=ScalarParameter() {defaultValue=0.02, parameterName=BoxRadius}) {materialFunction=&GeneratedRoundRect}
  n15 = Subtract(A=@n12, B=Subtract(A=@n14.Result, B=@n7.Result))
  FunctionOutput(None=LinearInterpolate(A=LinearInterpolate(A=@n1.RGB, B=@n2.RGB, Alpha=@n7.Result), B=Multiply(A=@n12, B=@n13.RGB), Alpha=Clamp(None=@n15))) {outputName=Color}
  FunctionOutput(None=@n14.Result) {outputName=Mask}
 ]
}
