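# Material function that draws a glowing curved line between P0 and P1 for AudioWidgets Slate materials, outputting Color and Mask.
# A Custom HLSL node loops over CurveSteps segments of a power curve, taking the minimum segment distance field.
# UVs are center-scaled and rotated around the endpoints' midpoint, with UseAutoFlip adding a half turn below 0.5 CurveAmount to mirror the bend.
# Two smoothsteps form the glow, also the Mask, and a tinted center, combined with Max.
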MaterialFunction /AudioWidgets/AudioMaterialSlate/Functions/MF_CurvedLine
graph {
 properties {bExposeToLibrary=true}
 nodes [
  n1 = Divide(A=FunctionInput(Preview=Constant() {r=0.01}) {bUsePreviewValueAsDefault=true, inputName=LineGlowThickness, inputType=FunctionInput_Scalar}) {constB=100}
  n2 = FunctionInput(Preview=Constant2Vector() {g=1}) {bUsePreviewValueAsDefault=true, inputName=P0, inputType=FunctionInput_Vector2, sortPriority=2}
  MaterialExpressionReroute_0 = Reroute(None=@n2)
  n3 = FunctionInput(Preview=Constant2Vector() {r=1}) {bUsePreviewValueAsDefault=true, inputName=P1, inputType=FunctionInput_Vector2, sortPriority=2}
  MaterialExpressionReroute_1 = Reroute(None=@n3)
  n4 = TextureCoordinate()
  n5 = Constant2Vector() {g=0.8, r=0.8}
  n6 = FunctionInput(Preview=Constant() {r=0.2}) {bUsePreviewValueAsDefault=true, inputName="CurveAmount(Lerp)", inputType=FunctionInput_Scalar}
  n7 = Constant()
  n8 = If(A=@n6, "A > B"=@n7, "A == B"=@n7, "A < B"=Constant() {r=0.5}) {constB=0.5}
  n9 = FunctionInput(Preview=Constant()) {bUsePreviewValueAsDefault=true, inputName=RotateUV, inputType=FunctionInput_Scalar}
  n10 = FunctionInput(Preview=StaticBool() {value=true}) {bUsePreviewValueAsDefault=true, inputName=UseAutoFlip, inputType=FunctionInput_StaticBool}
  MaterialExpressionMaterialFunctionCall_2 = MaterialFunctionCall(
   UVs=MaterialFunctionCall(
    UVs=FunctionInput(Preview=@n4) {bUsePreviewValueAsDefault=true, inputName=UV, inputType=FunctionInput_Vector2}
    Texture Scale=FunctionInput(Preview=@n5) {bUsePreviewValueAsDefault=true, inputName=ScaleUVCenter, inputType=FunctionInput_Vector2}
   ) {materialFunction=&ScaleUVsByCenter}.UVs
   Rotation Center=Divide(A=Add(A=@n2, B=@n3))
   "Rotation Angle (0-1)"=StaticSwitch(True=Add(A=@n8, B=@n9), False=@n9, Value=@n10)
  ) {materialFunction=&CustomRotator}
  MaterialExpressionDivide_0 = Divide(A=FunctionInput(Preview=Constant() {r=0.1}) {bUsePreviewValueAsDefault=true, inputName=Thickness, inputType=FunctionInput_Scalar}) {constB=100}
  MaterialExpressionFunctionInput_5 = FunctionInput(Preview=Constant() {r=32}) {bUsePreviewValueAsDefault=true, description=Default 32, inputName=CurveSteps, inputType=FunctionInput_Scalar}
  n11 = LinearInterpolate(Alpha=@n6) {constA=-1}
  MaterialExpressionLinearInterpolate_6 = LinearInterpolate(
   A=FunctionInput(Preview=Constant() {r=0.2}) {bUsePreviewValueAsDefault=true, inputName=CurveMin, inputType=FunctionInput_Scalar}
   B=FunctionInput(Preview=Constant() {r=4}) {bUsePreviewValueAsDefault=true, inputName=CurveMax, inputType=FunctionInput_Scalar}
   Alpha=StaticSwitch(True=Abs(None=@n11), False=LinearInterpolate(Alpha=@n6), Value=@n10)
  ) {constA=1, constB=10}
  n12 = Custom!(
   P0=@MaterialExpressionReroute_0
   P1=@MaterialExpressionReroute_1
   uv=@MaterialExpressionMaterialFunctionCall_2.Rotated Values
   Thickness=@MaterialExpressionDivide_0
   steps=@MaterialExpressionFunctionInput_5
   Exp=@MaterialExpressionLinearInterpolate_6
  ) {
   code="float3 col = float3(1.0,1.0,1.0);\r\n\r\nfor (int i = 0; i < steps; i++) {\r\n    float dx = P1.x-P0.x;\r\n    float dy = P1.y-P0.y;\r\n\r\n    float2 vert0 = float2(float(i)/float(steps),float(i)/float(steps));// get vertex_0 of segment per step\r\n    float2 vert1 = float2(float(i+1)/float(steps),float(i+1)/float(steps));// get vertex_1 of segment per step\r\n    vert0 = float2(vert0.x,pow(vert0.y,exponent)); // apply our function on y\r\n    vert1 = float2(vert1.x,pow(vert1.y,exponent)); // apply our function on y\r\n    vert0 = float2(vert0.x*dx,vert0.y*dy); // rescale to P0, P1 bounds\r\n    vert1 = float2(vert1.x*dx,vert1.y*dy); // rescale to P0, P1 bounds\r\n    vert0 = float2(vert0.x+P0.x,vert0.y+P0.y); // offset by P0\r\n    vert1 = float2(vert1.x+P0.x,vert1.y+P0.y); // offset by P0\r\n\r\n    float2 dvec_0p = uv - vert0;  // vector from vert0 to current pixel\r\n    float2 dvec_01 = vert1 - vert0;  // vector from vert0 to vert1 \r\n\r\n    float vector_projection_scalar = dot(dvec_0p, dvec_01) / dot(dvec_01, dvec_01); //gets the projection vector scalar of dvec_0p on dvec_01\r\n    float vproj_scalar_0p_01 = clamp(vector_projection_scalar,0.0,1.0);  // clamp it, to only get the line as a segment (instead of an infinite line)\r\n    float2 vproj_0p_01 = vproj_scalar_0p_01 * dvec_01;  // the projected vector of dvec_0p on dvec_01\r\n    float dist_segment = distance(dvec_0p, vproj_0p_01) - Thickness;  // Distance field\r\n    col = min(col,smoothstep(0.0,1.0, dist_segment));  // smoothes the field and adds all segments together\r\n}\r\n\r\nreturn col;"
   inputs [
    {inputName=P0, input {expression=&self:MaterialExpressionReroute_0}}
    {inputName=P1, input {expression=&self:MaterialExpressionReroute_1}}
    {inputName=uv, input {expression=&self:MaterialExpressionMaterialFunctionCall_2}}
    {inputName=Thickness, input {expression=&self:MaterialExpressionDivide_0}}
    {inputName=steps, input {expression=&self:MaterialExpressionFunctionInput_5}}
    {inputName=exponent, input {expression=&self:MaterialExpressionLinearInterpolate_6}}
   ]
  }
  n13 = OneMinus(None=Saturate(None=@n12))
  n14 = SmoothStep(Min=OneMinus(None=@n1), Value=@n13) {constMin=0.9}
  n15 = FunctionInput(Preview=Constant3Vector() {constant=<rgba 0.2747 0.1834 1 0>}) {bUsePreviewValueAsDefault=true, inputName=LineColor}
  n16 = SmoothStep(Min=OneMinus(None=Divide(A=@n1) {constB=5}), Value=@n13) {constMin=0.9}
  n17 = Constant() {r=0.2}
  n18 = Clamp(None=Multiply(
    A=@n15
    B=FunctionInput(Preview=@n17) {bUsePreviewValueAsDefault=true, description=Default 16, inputName=LineCenterBrightness, inputType=FunctionInput_Scalar}
   ) {constB=1.5})
  FunctionOutput(None=Max(A=Multiply(A=@n14, B=@n15), B=Multiply(A=@n16, B=@n18))) {outputName=Color}
  FunctionOutput(None=@n14) {outputName=Mask}
 ]
}
