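# Material function in AudioMaterialSlate that draws a line segment between two 2D points, with a separate glow output.
# It projects the coordinates, defaulting to UVs, onto the clamped segment and subtracts the line width from the distance, forming a signed distance field.
# The main output turns that distance into a mask with a narrow fixed falloff.
# The Glow output uses a width-scaled falloff, shaped by SmoothStep around GlowThresh ± GlowThick.
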MaterialFunction /AudioWidgets/AudioMaterialSlate/Functions/MF_DrawLineSimple
graph {
 properties {bExposeToLibrary=true, description="Draws a line between 2 points using supplied 2D coordinates. "}
 nodes [
  n1 = FunctionInput() {bUsePreviewValueAsDefault=true, inputName=P0, inputType=FunctionInput_Vector2, sortPriority=1}
  n2 = Subtract(
   A=FunctionInput(Preview=TextureCoordinate() {desc=UVs}) {bUsePreviewValueAsDefault=true, inputName=Coordinates, inputType=FunctionInput_Vector2}
   B=@n1
  ) {desc=PA}
  n3 = FunctionInput() {bUsePreviewValueAsDefault=true, inputName=P1, inputType=FunctionInput_Vector2, sortPriority=2, previewValue=<1 1 0 1>}
  n4 = Subtract(A=@n3, B=@n1) {desc=BA}
  n5 = DotProduct(A=@n2, B=@n4)
  n6 = DotProduct(A=@n4, B=@n4)
  n7 = Multiply(A=Clamp(None=Divide(A=@n5, B=@n6) {constB=1}), B=@n4)
  n8 = FunctionInput() {bUsePreviewValueAsDefault=true, inputName=Line Width, inputType=FunctionInput_Scalar, sortPriority=3, previewValue=<0.005 0 0 1>}
  n9 = Subtract(A=Distance(A=Subtract(A=@n2, B=@n7), B=Constant()), B=@n8)
  n10 = Divide(A=@n9) {constB=0.003}
  FunctionOutput(None=OneMinus(None=Clamp(None=@n10)))
  n11 = FunctionInput(Preview=Constant() {r=0.5}) {bUsePreviewValueAsDefault=true, inputName=GlowThresh, inputType=FunctionInput_Scalar}
  n12 = FunctionInput(Preview=Constant() {r=0.1}) {bUsePreviewValueAsDefault=true, inputName=GlowThick, inputType=FunctionInput_Scalar}
  n13 = Clamp(None=Divide(A=@n9, B=Multiply(A=@n8) {constB=30}))
  FunctionOutput(None=SmoothStep(Min=Subtract(A=@n11, B=@n12), Max=Add(A=@n11, B=@n12), Value=OneMinus(None=@n13))) {outputName=Glow}
 ]
}
