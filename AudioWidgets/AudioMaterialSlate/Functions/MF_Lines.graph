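# MF_Lines is a library-exposed material function in the AudioWidgets Slate material set that generates repeating vertical line patterns from UVs.
# It multiplies the UV X component by NumLinesX and takes the result modulo one, producing a repeating sawtooth.
# The Lines output thresholds that sawtooth at its midpoint into hard binary stripes.
# The Glow output smoothsteps the same sawtooth between GlowMin and GlowMax for a soft per-line gradient.
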MaterialFunction /AudioWidgets/AudioMaterialSlate/Functions/MF_Lines
graph {
 properties {bExposeToLibrary=true}
 nodes [
  n1 = FunctionInput(Preview=TextureCoordinate()) {bUsePreviewValueAsDefault=true, inputName=UV, inputType=FunctionInput_Vector2}
  n2 = FunctionInput(Preview=Constant() {r=9}) {bUsePreviewValueAsDefault=true, inputName=NumLinesX, inputType=FunctionInput_Scalar, sortPriority=1, previewValue=<0 0 0 2>}
  n3 = Fmod(
   A=Multiply(A=MaterialFunctionCall(Float2=@n1) {materialFunction=&BreakOutFloat2Components}.R, B=@n2)
   B=Constant() {r=1}
  )
  n4 = Constant()
  FunctionOutput(None=If(A=@n3, "A > B"=Constant() {r=1}, "A == B"=@n4, "A < B"=@n4) {constB=0.5}) {outputName=Lines}
  n5 = Constant()
  n6 = Constant() {r=1}
  FunctionOutput(None=SmoothStep(
    Min=FunctionInput(Preview=@n5) {bUsePreviewValueAsDefault=true, inputName=GlowMin, inputType=FunctionInput_Scalar, sortPriority=5}
    Max=FunctionInput(Preview=@n6) {bUsePreviewValueAsDefault=true, inputName=GlowMax, inputType=FunctionInput_Scalar, sortPriority=6}
    Value=@n3
   )) {outputName=Glow}
 ]
}
