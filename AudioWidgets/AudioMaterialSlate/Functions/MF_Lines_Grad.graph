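# AudioWidgets material function, exposed to the library, that draws a mask of evenly spaced horizontal lines with adjustable count, thickness and length.
# It multiplies V by NumLinesX and wraps the result with Fmod to form repeating cells along the vertical UV axis.
# Nested If nodes output 1 only where the cell position is below LineThick and U is below LineLength, giving hard-edged lines anchored at U = 0.
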MaterialFunction /AudioWidgets/AudioMaterialSlate/Functions/MF_Lines_Grad
graph {
 properties {bExposeToLibrary=true}
 nodes [
  n1 = MaterialFunctionCall(Float2=FunctionInput(Preview=TextureCoordinate()) {bUsePreviewValueAsDefault=true, inputName=UV, inputType=FunctionInput_Vector2}) {materialFunction=&BreakOutFloat2Components}
  n2 = FunctionInput(Preview=Constant() {r=9}) {bUsePreviewValueAsDefault=true, inputName=NumLinesX, inputType=FunctionInput_Scalar, sortPriority=1, previewValue=<0 0 0 2>}
  n3 = Multiply(A=@n1.G, B=@n2)
  n4 = Constant() {r=1}
  n5 = FunctionInput() {bUsePreviewValueAsDefault=true, inputName=LineThick, inputType=FunctionInput_Scalar, previewValue=<0.5 0.5 0.5 0.5>}
  n6 = Constant()
  n7 = FunctionInput() {bUsePreviewValueAsDefault=true, inputName=LineLength, inputType=FunctionInput_Scalar, previewValue=<0.5 0.5 0.5 0.5>}
  n8 = Constant()
  n9 = Constant() {r=1}
  FunctionOutput(None=If(A=Fmod(A=@n3, B=@n4), B=@n5, "A > B"=@n6, "A == B"=@n6, "A < B"=If(A=@n1.R, B=@n7, "A > B"=@n8, "A == B"=@n8, "A < B"=@n9)) {constB=0.5}) {outputName=Lines}
 ]
}
