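# UI material for the AudioWidgets button, drawing a procedural, translucent round button with bevel, highlight and shadow.
# Sphere gradients mask the body, bevel ring and center, whose two-color shading comes from rotated UV gradients.
# Offset soft gradients add a highlight and shadow behind the body and join its mask in a summed alpha.
# The Click parameter drives the press look by changing UV scale, bevel gradient strength and main color brightness.
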Material /AudioWidgets/AudioMaterialSlate/M_AudioMaterialButton
graph {
 properties {blendMode=BLEND_Translucent, materialDomain=MD_UI, pixelDepthOffsetMode=PDOM_Legacy, usedShadingModels=Default Lit}
 nodes [
  n1 = Constant() {r=0.1}
  n2 = ScalarParameter() {defaultValue=0.1, parameterName=AA}
  n3 = MaterialFunctionCall(UVs=TextureCoordinate(), Texture Scale=LinearInterpolate(Alpha=ScalarParameter() {parameterName=Click, sliderMax=1}) {constA=1, constB=0.9}) {materialFunction=&ScaleUVsByCenter}
  n4 = ScalarParameter() {defaultValue=0.28, parameterName="Radius (S)"}
  n5 = SmoothStep(
   Min=Subtract(A=@n1, B=@n2)
   Max=Add(A=@n1, B=@n2)
   Value=MaterialFunctionCall!(UVs=@n3.UVs, CenterPosition, Radius=@n4) {materialFunction=/Engine/Functions/Engine_MaterialFunctions01/Gradient/SphereGradient-2D}.Result 0-1
  )
  n6 = ScalarParameter() {defaultValue=1.6808, parameterName=SmoothBevelX}
  n7 = ScalarParameter() {defaultValue=0.5, parameterName=SmoothBevelY}
  n8 = SmoothStep(Value=MaterialFunctionCall(
    UVs=TextureCoordinate()
    CenterPosition=AppendVector(A=@n6, B=@n7)
    Radius=ScalarParameter() {defaultValue=0.35, parameterName=SmoothBevelSize}
   ) {materialFunction=&SphereGradient-2D}.Result 0-1)
  n9 = ScalarParameter() {defaultValue=2, parameterName=SmoothBevelContrast}
  n10 = Multiply(A=Clamp(None=Power(Base=@n8, Exp=@n9)), B=ScalarParameter() {defaultValue=0.7, parameterName=SmoothBevelInt})
  n11 = ScalarParameter() {defaultValue=0.5, parameterName=ShadowOffsetX}
  n12 = ScalarParameter() {defaultValue=0.5, parameterName=ShadowOffsetY}
  n13 = SmoothStep(Value=MaterialFunctionCall(UVs=TextureCoordinate(), CenterPosition=AppendVector(A=@n11, B=@n12), Radius=ScalarParameter() {defaultValue=0.35, parameterName=ShadowSize}) {materialFunction=&SphereGradient-2D}.Result 0-1)
  n14 = ScalarParameter() {defaultValue=2, parameterName=ShadowContrast}
  n15 = Multiply(A=Clamp(None=Power(Base=@n13, Exp=@n14)), B=ScalarParameter() {defaultValue=0.7, parameterName=ShadowIntensity})
  Add(A=@n5, B=Clamp(None=Add(A=@n10, B=@n15)))
  n16 = Multiply(A=@n10, B=VectorParameter() {parameterName=SmoothBevelColor, defaultValue=<rgba 0.9427 0.9427 0.9427 1>}.RGB)
  n17 = Multiply(A=@n15, B=VectorParameter() {parameterName=ShadowColor, defaultValue=<rgba 0 0 0 1>}.RGB)
  n18 = VectorParameter() {group=Design, parameterName=Color_1, defaultValue=<rgba 0.95 0.95 0.95 1>}
  n19 = VectorParameter() {group=Design, parameterName=Color_2, defaultValue=<rgba 0.1266 0.1387 0.15 1>}
  n20 = ScalarParameter() {group=Gradient, parameterName=EdgeBevelRot}
  n21 = ScalarParameter() {defaultValue=0.2, group=Gradient, parameterName=Power}
  n22 = Power(
   Base=MaterialFunctionCall(Float2=MaterialFunctionCall(UVs=@n3.UVs, "Rotation Angle (0-1)"=@n20) {materialFunction=&CustomRotator}.Rotated Values) {materialFunction=&BreakOutFloat2Components}.R
   Exp=@n21
  )
  n23 = LinearInterpolate(
   A=ScalarParameter() {defaultValue=0.8, parameterName=BevelClickMin}
   B=ScalarParameter() {defaultValue=0.04, parameterName=BevelClickMax}
   Alpha=ScalarParameter() {parameterName=Click}
  ) {constA=0.8, constB=0.04}
  n24 = ScalarParameter() {defaultValue=0.5, parameterName=BevelRot}
  n25 = Power(
   Base=MaterialFunctionCall(Float2=MaterialFunctionCall(UVs=@n3.UVs, "Rotation Angle (0-1)"=@n24) {materialFunction=&CustomRotator}.Rotated Values) {materialFunction=&BreakOutFloat2Components}.R
   Exp=@n21
  )
  n26 = Constant() {r=0.5}
  n27 = ScalarParameter() {defaultValue=0.392, parameterName=BevelSoftness}
  n28 = Subtract(A=@n26, B=@n27)
  n29 = Add(A=@n26, B=@n27)
  n30 = ScalarParameter() {defaultValue=1.6724, group=Bevel, parameterName=BevelSize}
  n31 = MaterialFunctionCall(UVs=@n3.UVs, Radius=Subtract(A=@n4, B=Multiply(A=@n30) {constB=0.01})) {materialFunction=&SphereGradient-2D}
  n32 = LinearInterpolate(
   A=@n18.RGB
   B=@n19.RGB
   Alpha=LinearInterpolate(A=Multiply(A=@n22, B=@n23), B=Multiply(A=@n25, B=@n23), Alpha=SmoothStep(Min=@n28, Max=@n29, Value=@n31.Result 0-1))
  )
  n33 = Multiply(
   A=VectorParameter() {parameterName=MainColor, defaultValue=<rgba 1 1 1 1>}.RGB
   B=LinearInterpolate(Alpha=ScalarParameter() {parameterName=Click, sliderMax=1}) {constA=1, constB=2}
  )
  n34 = MaterialFunctionCall(Float2=MaterialFunctionCall(UVs=@n3.UVs, "Rotation Angle (0-1)"=ScalarParameter() {defaultValue=0.75, parameterName=Centerrot}) {materialFunction=&CustomRotator}.Rotated Values) {materialFunction=&BreakOutFloat2Components}
  n35 = ScalarParameter() {defaultValue=4, parameterName=CenterPow}
  n36 = LinearInterpolate(A=@n19.RGB, B=@n18.RGB, Alpha=Multiply(A=Power(Base=@n34.R, Exp=@n35), B=ScalarParameter() {defaultValue=1.1, parameterName=CenterMult}))
  n37 = ScalarParameter() {defaultValue=3.7153, parameterName=CenterSize}
  n38 = MaterialFunctionCall(UVs=@n3.UVs, Radius=Subtract(A=@n4, B=Multiply(A=@n37) {constB=0.01})) {materialFunction=&SphereGradient-2D}
  LinearInterpolate(
   A=Subtract(A=Clamp(None=@n16), B=Clamp(None=@n17))
   B=LinearInterpolate(A=Multiply(A=@n32, B=@n33), B=Multiply(A=@n36, B=@n33), Alpha=SmoothStep(Min=@n28, Max=@n29, Value=@n38.Result 0-1))
   Alpha=@n5
  )
  ScalarParameter() {parameterName=Brightness1}
  ScalarParameter() {parameterName=Brightness2}
 ]
}
