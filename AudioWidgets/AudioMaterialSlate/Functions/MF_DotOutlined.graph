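# MF_DotOutlined is a library material function in AudioWidgets that draws an antialiased dot with a colored outline for audio Slate materials.
# A 2D sphere gradient from UV, center and radius is thresholded by smoothsteps whose width OutlineSmoothness sets.
# Subtracting two thresholds, offset by OutlineThickness, forms the ring that blends CenterColor into OutlineColor.
# A fixed-inset threshold trims the outer edge, masking the color and feeding MaskOutput.
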MaterialFunction /AudioWidgets/AudioMaterialSlate/Functions/MF_DotOutlined
graph {
 properties {bExposeToLibrary=true}
 nodes [
  n1 = FunctionInput(Preview=Constant3Vector() {constant=<rgba 0 1 0.1748 0>}) {bUsePreviewValueAsDefault=true, inputName=CenterColor}
  n2 = FunctionInput(Preview=Constant3Vector() {constant=<rgba 1 0 0 0>}) {bUsePreviewValueAsDefault=true, inputName=OutlineColor}
  n3 = Constant() {r=0.5}
  n4 = FunctionInput(Preview=Constant() {r=0.008}) {bUsePreviewValueAsDefault=true, inputName=OutlineSmoothness, inputType=FunctionInput_Scalar}
  n5 = Subtract(A=@n3, B=@n4)
  n6 = Add(A=@n3, B=@n4)
  n7 = MaterialFunctionCall!(
   UVs=FunctionInput(Preview=TextureCoordinate()) {bUsePreviewValueAsDefault=true, inputName=UV, inputType=FunctionInput_Vector2}
   CenterPosition=FunctionInput(Preview=Constant2Vector() {g=0.5, r=0.5}) {bUsePreviewValueAsDefault=true, inputName=Center, inputType=FunctionInput_Vector2}
   Radius=FunctionInput(Preview=Constant() {r=0.5}) {bUsePreviewValueAsDefault=true, inputName=Radius, inputType=FunctionInput_Scalar}
  ) {materialFunction=/Engine/Functions/Engine_MaterialFunctions01/Gradient/SphereGradient-2D}
  n8 = Subtract(
   A=@n7.Result 0-1
   B=FunctionInput(Preview=Constant() {r=0.145}) {bUsePreviewValueAsDefault=true, inputName=OutlineThickness, inputType=FunctionInput_Scalar}
  )
  n9 = Clamp(None=Subtract(A=SmoothStep(Min=@n5, Max=@n6, Value=@n7.Result 0-1), B=SmoothStep(Min=@n5, Max=@n6, Value=@n8)))
  n10 = SmoothStep(Min=@n5, Max=@n6, Value=Subtract(A=@n7.Result 0-1) {constB=0.07})
  FunctionOutput(None=LinearInterpolate(B=LinearInterpolate(A=@n1, B=@n2, Alpha=@n9), Alpha=@n10)) {outputName=Color}
  FunctionOutput(None=@n10) {outputName=MaskOutput}
 ]
}
