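# UI material that draws an ADSR envelope graph for AudioWidgets, inside an outlined rounded box.
# Attack, Decay, Sustain and Release parameters place four points, with release starting directly at the decay endpoint, so sustain only sets height.
# Custom HLSL draws each of the three segments as a stepped power-curve distance field, rotating UVs 180 degrees to flip curvature.
# Glow and core lines plus MF_DotOutlined handles are composited over the box.
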Material /AudioWidgets/AudioMaterialSlate/M_AudioMaterialEnvelope_ADSR
graph {
 properties {blendMode=BLEND_Masked, materialDomain=MD_UI, pixelDepthOffsetMode=PDOM_Legacy, usedShadingModels=Default Lit}
 nodes [
  n1 = MaterialFunctionCall(UVs=TextureCoordinate(), Texture Scale=Constant2Vector() {g=2, r=1}) {materialFunction=&ScaleUVsByCenter}
  n2 = ScalarParameter() {defaultValue=0.9609, parameterName=BoxSize}
  n3 = AppendVector(A=@n2, B=Multiply(A=@n2) {constB=0.5})
  n4 = Subtract(A=@n3, B=ScalarParameter() {defaultValue=0.018, parameterName=BoxBorderSize})
  n5 = ScalarParameter() {defaultValue=0.01, parameterName=CornerRadius}
  n6 = ScalarParameter() {defaultValue=56.137, parameterName=BoxSharpness}
  n7 = ScalarParameter() {defaultValue=0.01, parameterName=ThicknessGlow}
  MaterialExpressionClamp_4 = Clamp(None=AppendVector(A=Constant() {r=0.06}, B=Constant() {r=0.99})) {maxDefault=0.95, minDefault=0.07}
  n8 = LinearInterpolate(Alpha=ScalarParameter() {defaultValue=0.25, group=Attack, parameterName=A_Time, sliderMax=1}) {constA=0.05, constB=0.95}
  n9 = OneMinus(None=LinearInterpolate(Alpha=ScalarParameter() {defaultValue=0.8, group=Attack, parameterName=A_Int, sliderMax=1}) {constA=0.05, constB=0.95})
  MaterialExpressionClamp_2 = Clamp(None=AppendVector(A=@n8, B=@n9)) {maxDefault=0.95, minDefault=0.07}
  n10 = MaterialFunctionCall(
   UVs=Multiply(A=TextureCoordinate(), B=Constant2Vector() {g=1, r=2})
   Texture Scale=ScalarParameter() {defaultValue=0.9, parameterName=MainUVScale}
  ) {materialFunction=&ScaleUVsByCenter}
  n11 = ScalarParameter() {defaultValue=1, group=Attack, parameterName=A_Curve, sliderMax=1}
  n12 = Constant()
  MaterialExpressionMaterialFunctionCall_2 = MaterialFunctionCall(
   UVs=@n10.UVs
   Rotation Center=Divide(A=Add(A=@MaterialExpressionClamp_4, B=@MaterialExpressionClamp_2))
   "Rotation Angle (0-1)"=If(A=@n11, "A > B"=@n12, "A == B"=@n12, "A < B"=Constant() {r=0.5}) {constB=0.5}
  ) {materialFunction=&CustomRotator}
  MaterialExpressionScalarParameter_46 = ScalarParameter() {defaultValue=0.01, parameterName=Thickness}
  MaterialExpressionScalarParameter_49 = ScalarParameter() {defaultValue=16, parameterName=Steps}
  n13 = LinearInterpolate(Alpha=@n11) {constA=-1}
  n14 = Constant() {r=20}
  MaterialExpressionClamp_7 = Clamp(None=LinearInterpolate(Alpha=Abs(None=@n13)) {constA=1, constB=10}, Min=Divide(B=@n14), Max=@n14) {minDefault=0.05}
  n15 = Custom!(
   P0=@MaterialExpressionClamp_4
   P1=@MaterialExpressionClamp_2
   uv=@MaterialExpressionMaterialFunctionCall_2.Rotated Values
   Thickness=@MaterialExpressionScalarParameter_46
   steps=@MaterialExpressionScalarParameter_49
   Exp=@MaterialExpressionClamp_7
  ) {
   code="float3 col = float3(1.0,1.0,1.0);\r\n\r\nfor (int i = 0; i < steps; i++) {\r\n    float dx = P1.x-P0.x;\r\n    float dy = P1.y-P0.y;\r\n\r\n    float2 vert0 = float2(float(i)/float(steps),float(i)/float(steps));// get vertex_0 of segment per step\r\n    float2 vert1 = float2(float(i+1)/float(steps),float(i+1)/float(steps));// get vertex_1 of segment per step\r\n    vert0 = float2(vert0.x,pow(vert0.y,exponent)); // apply our function on y\r\n    vert1 = float2(vert1.x,pow(vert1.y,exponent)); // apply our function on y\r\n    vert0 = float2(vert0.x*dx,vert0.y*dy); // rescale to P0, P1 bounds\r\n    vert1 = float2(vert1.x*dx,vert1.y*dy); // rescale to P0, P1 bounds\r\n    vert0 = float2(vert0.x+P0.x,vert0.y+P0.y); // offset by P0\r\n    vert1 = float2(vert1.x+P0.x,vert1.y+P0.y); // offset by P0\r\n\r\n    float2 dvec_0p = uv - vert0;  // vector from vert0 to current pixel\r\n    float2 dvec_01 = vert1 - vert0;  // vector from vert0 to vert1 \r\n\r\n    float vector_projection_scalar = dot(dvec_0p, dvec_01) / dot(dvec_01, dvec_01); //gets the projection vector scalar of dvec_0p on dvec_01\r\n    float vproj_scalar_0p_01 = clamp(vector_projection_scalar,0.0,1.0);  // clamp it, to only get the line as a segment (instead of an infinite line)\r\n    float2 vproj_0p_01 = vproj_scalar_0p_01 * dvec_01;  // the projected vector of dvec_0p on dvec_01\r\n    float dist_segment = distance(dvec_0p, vproj_0p_01) - Thickness;  // Distance field\r\n    col = min(col,smoothstep(0.0,1.0, dist_segment));  // smoothes the field and adds all segments together\r\n}\r\n\r\nreturn col;"
   inputs [
    {inputName=P0, input {expression=&self:MaterialExpressionClamp_4}}
    {inputName=P1, input {expression=&self:MaterialExpressionClamp_2}}
    {inputName=uv, input {expression=&self:MaterialExpressionMaterialFunctionCall_2}}
    {inputName=Thickness, input {expression=&self:MaterialExpressionScalarParameter_46}}
    {inputName=steps, input {expression=&self:MaterialExpressionScalarParameter_49}}
    {inputName=exponent, input {expression=&self:MaterialExpressionClamp_7}}
   ]
  }
  n16 = OneMinus(None=Saturate(None=@n15))
  n17 = SmoothStep(Min=OneMinus(None=@n7), Value=@n16) {constMin=0.9}
  n18 = VectorParameter() {parameterName=MainColor, defaultValue=<rgba 0.2747 0.1834 1 0>}
  n19 = Multiply(A=@n17, B=@n18.RGB)
  n20 = Divide(A=@n7) {constB=5}
  n21 = Clamp(None=Multiply(A=@n18.RGB, B=ScalarParameter() {defaultValue=1.847, parameterName=CoreBrightness}) {constB=1.5})
  n22 = Multiply(A=SmoothStep(Min=OneMinus(None=@n20), Value=@n16) {constMin=0.9}, B=@n21)
  n23 = Add(A=LinearInterpolate(Alpha=ScalarParameter() {defaultValue=0.32, group=Decay, parameterName=D_Time, sliderMax=1}) {constB=0.95}, B=@n8)
  n24 = OneMinus(None=LinearInterpolate(Alpha=ScalarParameter() {defaultValue=0.344, group=Sustain, parameterName=S_Int, sliderMax=1}) {constA=0.05, constB=0.95})
  MaterialExpressionClamp_3 = Clamp(None=AppendVector(A=@n23, B=Clamp(None=@n24, Min=@n9) {maxDefault=0.95, minDefault=0.05})) {maxDefault=10, minDefault=0.07}
  n25 = ScalarParameter() {defaultValue=1, group=Decay, parameterName=D_Curve, sliderMax=1}
  n26 = Constant()
  MaterialExpressionMaterialFunctionCall_6 = MaterialFunctionCall(
   UVs=@n10.UVs
   Rotation Center=Divide(A=Add(A=@MaterialExpressionClamp_2, B=@MaterialExpressionClamp_3))
   "Rotation Angle (0-1)"=If(A=@n25, "A > B"=@n26, "A == B"=@n26, "A < B"=Constant() {r=0.5}) {constB=0.5}
  ) {materialFunction=&CustomRotator}
  n27 = LinearInterpolate(Alpha=@n25) {constA=-1}
  n28 = Constant() {r=20}
  MaterialExpressionClamp_6 = Clamp(None=LinearInterpolate(Alpha=Abs(None=@n27)) {constA=1, constB=10}, Min=Divide(B=@n28), Max=@n28) {minDefault=0.05}
  n29 = Custom(
   P0=@MaterialExpressionClamp_2
   P1=@MaterialExpressionClamp_3
   uv=@MaterialExpressionMaterialFunctionCall_6.Rotated Values
   Thickness=@MaterialExpressionScalarParameter_46
   steps=@MaterialExpressionScalarParameter_49
   Exp=@MaterialExpressionClamp_6
  ) {
   code="float3 col = float3(1.0,1.0,1.0);\r\n\r\nfor (int i = 0; i < steps; i++) {\r\n    float dx = P1.x-P0.x;\r\n    float dy = P1.y-P0.y;\r\n\r\n    float2 vert0 = float2(float(i)/float(steps),float(i)/float(steps));// get vertex_0 of segment per step\r\n    float2 vert1 = float2(float(i+1)/float(steps),float(i+1)/float(steps));// get vertex_1 of segment per step\r\n    vert0 = float2(vert0.x,pow(vert0.y,exponent)); // apply our function on y\r\n    vert1 = float2(vert1.x,pow(vert1.y,exponent)); // apply our function on y\r\n    vert0 = float2(vert0.x*dx,vert0.y*dy); // rescale to P0, P1 bounds\r\n    vert1 = float2(vert1.x*dx,vert1.y*dy); // rescale to P0, P1 bounds\r\n    vert0 = float2(vert0.x+P0.x,vert0.y+P0.y); // offset by P0\r\n    vert1 = float2(vert1.x+P0.x,vert1.y+P0.y); // offset by P0\r\n\r\n    float2 dvec_0p = uv - vert0;  // vector from vert0 to current pixel\r\n    float2 dvec_01 = vert1 - vert0;  // vector from vert0 to vert1 \r\n\r\n    float vector_projection_scalar = dot(dvec_0p, dvec_01) / dot(dvec_01, dvec_01); //gets the projection vector scalar of dvec_0p on dvec_01\r\n    float vproj_scalar_0p_01 = clamp(vector_projection_scalar,0.0,1.0);  // clamp it, to only get the line as a segment (instead of an infinite line)\r\n    float2 vproj_0p_01 = vproj_scalar_0p_01 * dvec_01;  // the projected vector of dvec_0p on dvec_01\r\n    float dist_segment = distance(dvec_0p, vproj_0p_01) - Thickness;  // Distance field\r\n    col = min(col,smoothstep(0.0,1.0, dist_segment));  // smoothes the field and adds all segments together\r\n}\r\n\r\nreturn col;"
   inputs [
    {inputName=P0, input {expression=&self:MaterialExpressionClamp_2}}
    {inputName=P1, input {expression=&self:MaterialExpressionClamp_3}}
    {inputName=uv, input {expression=&self:MaterialExpressionMaterialFunctionCall_6}}
    {inputName=Thickness, input {expression=&self:MaterialExpressionScalarParameter_46}}
    {inputName=steps, input {expression=&self:MaterialExpressionScalarParameter_49}}
    {inputName=exponent, input {expression=&self:MaterialExpressionClamp_6}}
   ]
  }
  n30 = OneMinus(None=Saturate(None=@n29))
  n31 = SmoothStep(Min=OneMinus(None=@n7), Value=@n30) {constMin=0.9}
  n32 = Multiply(A=@n31, B=@n18.RGB)
  n33 = Divide(A=@n7) {constB=5}
  n34 = Multiply(A=SmoothStep(Min=OneMinus(None=@n33), Value=@n30) {constMin=0.9}, B=@n21)
  n35 = ScalarParameter() {defaultValue=0.01, parameterName=ThicknessGlow}
  n36 = LinearInterpolate(Alpha=ScalarParameter() {defaultValue=0.2, group=Release, parameterName=R_Time, sliderMax=1}) {constB=0.95}
  n37 = Constant() {r=0.99}
  MaterialExpressionClamp_0 = Clamp(None=AppendVector(A=Add(A=@n36, B=@n23), B=Clamp(None=@n37) {maxDefault=0.95, minDefault=0.05})) {maxDefault=10, minDefault=0.07}
  n38 = ScalarParameter() {defaultValue=1, group=Release, parameterName=R_Curve, sliderMax=1}
  n39 = Constant()
  MaterialExpressionMaterialFunctionCall_1 = MaterialFunctionCall(
   UVs=@n10.UVs
   Rotation Center=Divide(A=Add(A=@MaterialExpressionClamp_3, B=@MaterialExpressionClamp_0))
   "Rotation Angle (0-1)"=If(A=@n38, "A > B"=@n39, "A == B"=@n39, "A < B"=Constant() {r=0.5}) {constB=0.5}
  ) {materialFunction=&CustomRotator}
  MaterialExpressionScalarParameter_3 = ScalarParameter() {defaultValue=0.01, parameterName=Thickness}
  MaterialExpressionScalarParameter_7 = ScalarParameter() {defaultValue=16, parameterName=Steps}
  n40 = LinearInterpolate(Alpha=@n38) {constA=-1}
  n41 = Constant() {r=20}
  MaterialExpressionClamp_1 = Clamp(None=LinearInterpolate(Alpha=Abs(None=@n40)) {constA=1, constB=10}, Min=Divide(B=@n41), Max=@n41) {minDefault=0.05}
  n42 = Custom(
   P0=@MaterialExpressionClamp_3
   P1=@MaterialExpressionClamp_0
   uv=@MaterialExpressionMaterialFunctionCall_1.Rotated Values
   Thickness=@MaterialExpressionScalarParameter_3
   steps=@MaterialExpressionScalarParameter_7
   Exp=@MaterialExpressionClamp_1
  ) {
   code="float3 col = float3(1.0,1.0,1.0);\r\n\r\nfor (int i = 0; i < steps; i++) {\r\n    float dx = P1.x-P0.x;\r\n    float dy = P1.y-P0.y;\r\n\r\n    float2 vert0 = float2(float(i)/float(steps),float(i)/float(steps));// get vertex_0 of segment per step\r\n    float2 vert1 = float2(float(i+1)/float(steps),float(i+1)/float(steps));// get vertex_1 of segment per step\r\n    vert0 = float2(vert0.x,pow(vert0.y,exponent)); // apply our function on y\r\n    vert1 = float2(vert1.x,pow(vert1.y,exponent)); // apply our function on y\r\n    vert0 = float2(vert0.x*dx,vert0.y*dy); // rescale to P0, P1 bounds\r\n    vert1 = float2(vert1.x*dx,vert1.y*dy); // rescale to P0, P1 bounds\r\n    vert0 = float2(vert0.x+P0.x,vert0.y+P0.y); // offset by P0\r\n    vert1 = float2(vert1.x+P0.x,vert1.y+P0.y); // offset by P0\r\n\r\n    float2 dvec_0p = uv - vert0;  // vector from vert0 to current pixel\r\n    float2 dvec_01 = vert1 - vert0;  // vector from vert0 to vert1 \r\n\r\n    float vector_projection_scalar = dot(dvec_0p, dvec_01) / dot(dvec_01, dvec_01); //gets the projection vector scalar of dvec_0p on dvec_01\r\n    float vproj_scalar_0p_01 = clamp(vector_projection_scalar,0.0,1.0);  // clamp it, to only get the line as a segment (instead of an infinite line)\r\n    float2 vproj_0p_01 = vproj_scalar_0p_01 * dvec_01;  // the projected vector of dvec_0p on dvec_01\r\n    float dist_segment = distance(dvec_0p, vproj_0p_01) - Thickness;  // Distance field\r\n    col = min(col,smoothstep(0.0,1.0, dist_segment));  // smoothes the field and adds all segments together\r\n}\r\n\r\nreturn col;"
   inputs [
    {inputName=P0, input {expression=&self:MaterialExpressionClamp_3}}
    {inputName=P1, input {expression=&self:MaterialExpressionClamp_0}}
    {inputName=uv, input {expression=&self:MaterialExpressionMaterialFunctionCall_1}}
    {inputName=Thickness, input {expression=&self:MaterialExpressionScalarParameter_3}}
    {inputName=steps, input {expression=&self:MaterialExpressionScalarParameter_7}}
    {inputName=exponent, input {expression=&self:MaterialExpressionClamp_1}}
   ]
  }
  n43 = OneMinus(None=Saturate(None=@n42))
  n44 = SmoothStep(Min=OneMinus(None=@n35), Value=@n43) {constMin=0.9}
  n45 = SmoothStep(Min=OneMinus(None=Divide(A=@n35) {constB=5}), Value=@n43) {constMin=0.9}
  n46 = Add(A=Add(A=Max(A=@n19, B=@n22), B=Max(A=@n32, B=@n34)), B=Max(A=Multiply(A=@n44, B=@n18.RGB), B=Multiply(A=@n45, B=@n21)))
  n47 = ScalarParameter() {parameterName=DotSMooth}
  n48 = ScalarParameter() {defaultValue=0.2, parameterName=DotOutlineThickness}
  n49 = Multiply(A=@n18.RGB, B=ScalarParameter() {defaultValue=1, parameterName=DotOutlineBrighness})
  n50 = Multiply(A=@n18.RGB, B=ScalarParameter() {defaultValue=0.2, parameterName=DotCenterBrighness})
  n51 = ScalarParameter() {defaultValue=0.096, parameterName=DotRadius}
  n52 = MaterialFunctionCall!(
   OutlineSmoothness=@n47
   UV=@n10.UVs
   Center=@MaterialExpressionClamp_4
   OutlineThickness=@n48
   OutlineColor=@n49
   CenterColor=@n50
   Radius=@n51
  ) {materialFunction=/AudioWidgets/AudioMaterialSlate/Functions/MF_DotOutlined}
  n53 = MaterialFunctionCall(
   OutlineSmoothness=@n47
   UV=@n10.UVs
   Center=@MaterialExpressionClamp_2
   OutlineThickness=@n48
   OutlineColor=@n49
   CenterColor=@n50
   Radius=@n51
  ) {materialFunction=&MF_DotOutlined}
  n54 = MaterialFunctionCall(
   OutlineSmoothness=@n47
   UV=@n10.UVs
   Center=@MaterialExpressionClamp_3
   OutlineThickness=@n48
   OutlineColor=@n49
   CenterColor=@n50
   Radius=@n51
  ) {materialFunction=&MF_DotOutlined}
  n55 = Add(A=Add(A=@n52.Color, B=@n53.Color), B=@n54.Color)
  n56 = MaterialFunctionCall(
   OutlineSmoothness=@n47
   UV=@n10.UVs
   Center=@MaterialExpressionClamp_0
   OutlineThickness=@n48
   OutlineColor=@n49
   CenterColor=@n50
   Radius=@n51
  ) {materialFunction=&MF_DotOutlined}
  n57 = Add(A=@n52.MaskOutput, B=@n53.MaskOutput)
  n58 = Clamp(None=Add(A=@n56.MaskOutput, B=Add(A=@n57, B=@n54.MaskOutput)))
  n59 = Add(A=Add(A=@n17, B=@n31), B=@n44)
  LinearInterpolate(
   A=LinearInterpolate(
    A=VectorParameter() {parameterName=BoxOutline, defaultValue=<rgba 0.1562 0.1562 0.1562 1>}.RGB
    B=VectorParameter() {parameterName=BoxBG, defaultValue=<rgba 0.0234 0.0234 0.0234 1>}.RGB
    Alpha=MaterialFunctionCall!(UV Coords=@n1.UVs, Box Dimensions=@n4, CornerRadius=@n5, Sharpness=@n6, Center) {materialFunction=/Engine/Functions/Engine_MaterialFunctions01/Gradient/GeneratedRoundRect}.Result
   )
   B=LinearInterpolate(A=Clamp(None=@n46), B=Add(A=@n55, B=@n56.Color), Alpha=@n58)
   Alpha=Clamp(None=Add(A=@n58, B=@n59))
  )
  MaterialFunctionCall(UV Coords=@n1.UVs, Box Dimensions=@n3, CornerRadius=ScalarParameter() {defaultValue=0.02, parameterName=BoxRadius}) {materialFunction=&GeneratedRoundRect}
 ]
}
