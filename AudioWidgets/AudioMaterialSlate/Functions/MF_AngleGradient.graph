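# AngleGradient material function in AudioWidgets' AudioMaterialSlate set, exposed to the library, outputs a 0-1 angular gradient around the UV center.
# Input UVs pass through CustomRotator, driven by a Rotate input, to set where the gradient starts.
# The centered components feed a fast arctangent with U as Y and V as X, measuring the angle from the vertical axis.
# The result is offset by pi and divided by two pi to normalize it.
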MaterialFunction /AudioWidgets/AudioMaterialSlate/Functions/MF_AngleGradient
graph {
 properties {bExposeToLibrary=true}
 nodes [
  n1 = TextureCoordinate()
  n2 = MaterialFunctionCall(Float2=MaterialFunctionCall(
    UVs=FunctionInput(Preview=@n1) {bUsePreviewValueAsDefault=true, inputName=UV, inputType=FunctionInput_Vector2}
    "Rotation Angle (0-1)"=FunctionInput() {bUsePreviewValueAsDefault=true, inputName=Rotate, inputType=FunctionInput_Scalar}
   ) {materialFunction=&CustomRotator}.Rotated Values) {materialFunction=&BreakOutFloat2Components}
  n3 = Add(A=Arctangent2Fast(Y=Subtract(A=@n2.R) {constB=0.5}, X=Subtract(A=@n2.G) {constB=0.5})) {constB=3.142}
  FunctionOutput(None=Divide(A=Divide(A=@n3)) {constB=3.142})
 ]
}
