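# MF_SingleLine_Smooth is an exposed AudioWidgets material function that outputs a soft-edged single-line mask at LinePosition.
# Two SmoothStep ramps run from the position to plus and minus LineThickness, scaled down to UV units; their sum is inverted.
# The reversed range on the second ramp fades the other side.
# Both Use V switch branches read the UV G component, so the switch has no effect.
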MaterialFunction /AudioWidgets/AudioMaterialSlate/Functions/MF_SingleLine_Smooth
graph {
 properties {bExposeToLibrary=true}
 nodes [
  n1 = FunctionInput(Preview=Constant() {r=0.5}) {bUsePreviewValueAsDefault=true, inputName=LinePosition, inputType=FunctionInput_Scalar}
  n2 = Multiply(A=FunctionInput(Preview=Constant() {r=29.0262}) {bUsePreviewValueAsDefault=true, inputName=LineThickness, inputType=FunctionInput_Scalar}) {constB=0.01}
  n3 = MaterialFunctionCall(Float2=FunctionInput(Preview=TextureCoordinate()) {bUsePreviewValueAsDefault=true, inputName=UV, inputType=FunctionInput_Vector2}) {materialFunction=&BreakOutFloat2Components}
  n4 = StaticSwitch(
   True=@n3.G
   False=@n3.G
   Value=FunctionInput(Preview=StaticBoolParameter() {parameterName=Use V}) {bUsePreviewValueAsDefault=true, inputName=Use V, inputType=FunctionInput_StaticBool}
  )
  n5 = SmoothStep(Min=@n1, Max=Add(A=@n1, B=@n2), Value=@n4)
  n6 = SmoothStep(Min=@n1, Max=Subtract(A=@n1, B=@n2), Value=@n4)
  FunctionOutput(None=OneMinus(None=Add(A=@n5, B=@n6)))
 ]
}
