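# Translucent UI material that draws a rotary knob for AudioWidgets, fully procedurally from radial and angular gradients with smoothstep edges, without textures.
# An angular mask driven by VALUE lights an LED arc with glow and a three-colour ramp over a range-limited bar track.
# Offset radial gradients add shadow and highlight around a bevelled body shaded by rotated two-colour gradients.
# VALUE also rotates the UVs that place the indicator dot on top.
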Material /AudioWidgets/AudioMaterialSlate/M_AudioMaterialKnob
graph {
 properties {blendMode=BLEND_Translucent, materialDomain=MD_UI, pixelDepthOffsetMode=PDOM_Legacy, usedShadingModels=Default Lit}
 nodes [
  n1 = ScalarParameter() {parameterName=FadeLedIN}
  n2 = ScalarParameter() {defaultValue=0.15, parameterName=FadeLedINmax}
  n3 = OneMinus(None=MaterialFunctionCall!(UV, Rotate=Constant() {r=0.4}) {materialFunction=/AudioWidgets/AudioMaterialSlate/Functions/MF_AngleGradient}.Result)
  n4 = LinearInterpolate(Alpha=ScalarParameter() {parameterName=VALUE, sliderMax=1}) {constB=0.8}
  n5 = Subtract(A=@n4, B=ScalarParameter() {parameterName=LedValueFalloff})
  n6 = Constant() {r=0.5}
  n7 = ScalarParameter() {parameterName=LED_Thickness}
  n8 = Add(A=@n6, B=@n7)
  n9 = ScalarParameter() {parameterName=LedAA}
  n10 = ScalarParameter() {parameterName=LED_Offset}
  n11 = MaterialFunctionCall!(UVs, CenterPosition, Radius=Add(A=ScalarParameter() {defaultValue=0.292, parameterName=LedBarRadius}, B=Multiply(A=@n10) {constB=0.1})) {materialFunction=/Engine/Functions/Engine_MaterialFunctions01/Gradient/SphereGradient-2D}
  n12 = Subtract(A=@n6, B=@n7)
  n13 = Max(
   A=SmoothStep(Min=@n8, Max=Add(A=@n8, B=@n9), Value=@n11.Result 0-1) {constMax=0.6, constMin=0.5}
   B=SmoothStep(Min=@n12, Max=Subtract(A=@n12, B=@n9), Value=@n11.Result 0-1) {constMax=0.6, constMin=0.4}
  )
  n14 = ScalarParameter() {defaultValue=0.296, parameterName=GlowOffsetRadius}
  n15 = Power(
   Base=MaterialFunctionCall!(
    LinePosition=ScalarParameter() {defaultValue=0.5, parameterName=GlowPosition}
    LineThickness=ScalarParameter() {defaultValue=24.5587, parameterName=GlowThickness}
    Use V
    UV=MaterialFunctionCall(Radius=@n14) {materialFunction=&SphereGradient-2D}.Result 0-1
   ) {materialFunction=/AudioWidgets/AudioMaterialSlate/Functions/MF_SingleLine_Smooth}.Result
   Exp=ScalarParameter() {defaultValue=1, parameterName=GlowContrast}
  )
  n16 = ScalarParameter() {defaultValue=0.488, parameterName=GlowIntensity}
  n17 = Multiply(
   A=Multiply(A=SmoothStep(Min=@n1, Max=@n2, Value=@n3) {constMax=0.15}, B=SmoothStep(Min=@n4, Max=@n5, Value=@n3))
   B=Add(A=OneMinus(None=@n13), B=Multiply(A=@n15, B=@n16))
  )
  n18 = MaterialFunctionCall() {materialFunction=&MF_AngleGradient}
  n19 = Constant() {r=1}
  n20 = If(A=@n18.Result, B=ScalarParameter() {defaultValue=0.4, parameterName=LedRangeMin}, "A > B"=@n19, "A == B"=@n19, "A < B"=Constant()) {constB=0.5}
  n21 = Constant()
  n22 = If(A=@n18.Result, B=ScalarParameter() {defaultValue=0.6, parameterName=LedRangeMax}, "A > B"=@n21, "A == B"=@n21, "A < B"=Constant() {r=1}) {constB=0.5}
  n23 = Constant() {r=0.5}
  n24 = ScalarParameter() {defaultValue=0.0441, parameterName=LedBarThickness, sliderMax=0.499}
  n25 = Add(A=@n23, B=@n24)
  n26 = ScalarParameter() {defaultValue=0.5, parameterName=LedBarAA}
  n27 = MaterialFunctionCall(Radius=ScalarParameter() {defaultValue=0.292, parameterName=LedBarRadius}) {materialFunction=&SphereGradient-2D}
  n28 = SmoothStep(Min=@n25, Max=Add(A=@n25, B=@n26), Value=@n27.Result 0-1) {constMax=0.6, constMin=0.5}
  n29 = Subtract(A=@n23, B=@n24)
  n30 = SmoothStep(Min=@n29, Max=Subtract(A=@n29, B=@n26), Value=@n27.Result 0-1) {constMax=0.6, constMin=0.4}
  n31 = Multiply(A=OneMinus(None=Multiply(A=@n20, B=@n22)), B=OneMinus(None=Max(A=@n28, B=@n30)))
  n32 = Add(A=@n31, B=@n17)
  n33 = ScalarParameter() {defaultValue=0.5, parameterName=ShadowOffsetX}
  n34 = ScalarParameter() {defaultValue=0.5, parameterName=ShadowOffsetY}
  n35 = SmoothStep(Value=MaterialFunctionCall(UVs=TextureCoordinate(), CenterPosition=AppendVector(A=@n33, B=@n34), Radius=ScalarParameter() {defaultValue=0.35, parameterName=ShadowSize}) {materialFunction=&SphereGradient-2D}.Result 0-1)
  n36 = ScalarParameter() {defaultValue=2, parameterName=ShadowContrast}
  n37 = Multiply(A=Clamp(None=Power(Base=@n35, Exp=@n36)), B=ScalarParameter() {defaultValue=0.7, parameterName=ShadowIntensity})
  n38 = ScalarParameter() {defaultValue=1.6808, parameterName=SmoothBevelX}
  n39 = ScalarParameter() {defaultValue=0.5, parameterName=SmoothBevelY}
  n40 = SmoothStep(Value=MaterialFunctionCall(
    UVs=TextureCoordinate()
    CenterPosition=AppendVector(A=@n38, B=@n39)
    Radius=ScalarParameter() {defaultValue=0.35, parameterName=SmoothBevelSize}
   ) {materialFunction=&SphereGradient-2D}.Result 0-1)
  n41 = ScalarParameter() {defaultValue=2, parameterName=SmoothBevelContrast}
  n42 = Multiply(A=Clamp(None=Power(Base=@n40, Exp=@n41)), B=ScalarParameter() {defaultValue=0.7, parameterName=SmoothBevelInt})
  n43 = Clamp(None=Add(A=@n42, B=@n37))
  n44 = Constant() {r=0.5}
  n45 = ScalarParameter() {defaultValue=0.392, parameterName=BevelSoftness}
  n46 = Subtract(A=@n44, B=@n45)
  n47 = Add(A=@n44, B=@n45)
  n48 = ScalarParameter() {defaultValue=0.312, parameterName="Radius (S)"}
  n49 = SmoothStep(
   Min=@n46
   Max=@n47
   Value=MaterialFunctionCall(Radius=@n48) {materialFunction=&SphereGradient-2D}.Result 0-1
  )
  Add(A=Add(A=Clamp(None=@n32), B=@n43), B=@n49)
  n50 = Add(A=Constant() {r=1}, B=ScalarParameter() {defaultValue=0.4, parameterName=RangeMin})
  n51 = ScalarParameter() {defaultValue=0.6, parameterName=RangeMax}
  n52 = ScalarParameter() {parameterName=VALUE, sliderMax=1}
  n53 = ScalarParameter() {defaultValue=0.7, group=Dot, parameterName=DotPosition}
  n54 = MaterialFunctionCall(
   UVs=MaterialFunctionCall(UVs=TextureCoordinate(), "Rotation Angle (0-1)"=LinearInterpolate(A=@n50, B=@n51, Alpha=@n52) {constA=-0.22}) {materialFunction=&CustomRotator}.Rotated Values
   CenterPosition=AppendVector(A=Constant() {r=0.5}, B=LinearInterpolate(Alpha=@n53) {constA=0.5, constB=0.1})
  ) {materialFunction=&SphereGradient-2D}
  n55 = LinearInterpolate(Alpha=ScalarParameter() {group=Dot, parameterName=DotSize}) {constA=0.999, constB=0.995}
  n56 = Constant() {r=1}
  If(A=@n54.Result 0-1, B=@n55, "A > B"=@n56, "A == B"=@n56, "A < B"=Constant()) {constB=0.995}
  n57 = ScalarParameter() {parameterName=DotAA}
  n58 = VectorParameter() {parameterName=BarShadowColor, defaultValue=<rgba 0.2917 0.0202 0.0057 1>}
  n59 = VectorParameter() {parameterName=BarColor, defaultValue=<rgba 0.2917 0.0202 0.0057 1>}
  n60 = SmoothStep(
   Min=ScalarParameter() {parameterName=BarShadowMin}
   Max=ScalarParameter() {defaultValue=1, parameterName=BarShadowMax}
   Value=MaterialFunctionCall(Radius=ScalarParameter() {defaultValue=0.284, parameterName=LedBarRadius}) {materialFunction=&SphereGradient-2D}.Result 0-1
  )
  n61 = MaterialFunctionCall(Rotate=Constant() {r=0.4}) {materialFunction=&MF_AngleGradient}
  n62 = MaterialFunctionCall(
   A=VectorParameter() {parameterName=LED_Min, defaultValue=<rgba 0.1946 0.9303 1 1>}.RGB
   B=VectorParameter() {parameterName=Led_Med, defaultValue=<rgba 0.1946 0.9303 1 1>}.RGB
   C=VectorParameter() {parameterName=Led_Max, defaultValue=<rgba 1 1 1 1>}.RGB
   Alpha=Power(Base=OneMinus(None=@n61.Result), Exp=ScalarParameter() {defaultValue=1, parameterName=GlowColorTransitionPow})
  ) {materialFunction=&Lerp_3Color}
  n63 = VectorParameter() {parameterName=LedTint, defaultValue=<rgba 1 1 1 1>}
  n64 = LinearInterpolate(
   A=Multiply(A=LinearInterpolate(A=@n58.RGB, B=@n59.RGB, Alpha=@n60), B=@n31)
   B=Multiply(A=Multiply(A=@n62.Result, B=@n63.RGB), B=ScalarParameter() {defaultValue=1, parameterName=LED_Total_Int})
   Alpha=@n17
  )
  n65 = Clamp(None=Multiply(A=@n42, B=VectorParameter() {parameterName=SmoothBevelColor, defaultValue=<rgba 0.9427 0.9427 0.9427 1>}.RGB))
  n66 = Clamp(None=Multiply(A=@n37, B=VectorParameter() {parameterName=ShadowColor, defaultValue=<rgba 0 0 0 1>}.RGB))
  n67 = LinearInterpolate(A=VectorParameter() {parameterName=EdgeFillColor, defaultValue=<rgba 0 0 0 1>}.RGB, B=Clamp(None=Subtract(A=@n65, B=@n66)), Alpha=@n43)
  n68 = VectorParameter() {group=Design, parameterName=Color_1, defaultValue=<rgba 0.1406 0.1619 0.1979 1>}
  n69 = VectorParameter() {group=Design, parameterName=Color_2, defaultValue=<rgba 0.0088 0.0096 0.0104 1>}
  n70 = TextureCoordinate()
  n71 = ScalarParameter() {group=Gradient, parameterName=EdgeBevelRot}
  n72 = ScalarParameter() {defaultValue=0.2, group=Gradient, parameterName=Power}
  n73 = Power(
   Base=MaterialFunctionCall(Float2=MaterialFunctionCall(UVs=@n70, "Rotation Angle (0-1)"=@n71) {materialFunction=&CustomRotator}.Rotated Values) {materialFunction=&BreakOutFloat2Components}.R
   Exp=@n72
  )
  n74 = ScalarParameter() {defaultValue=1.1, group=Gradient, parameterName=MultGradient}
  n75 = ScalarParameter() {defaultValue=0.5, parameterName=BevelRot}
  n76 = Power(
   Base=MaterialFunctionCall(Float2=MaterialFunctionCall(UVs=@n70, "Rotation Angle (0-1)"=@n75) {materialFunction=&CustomRotator}.Rotated Values) {materialFunction=&BreakOutFloat2Components}.R
   Exp=@n72
  )
  n77 = ScalarParameter() {defaultValue=0.7849, group=Bevel, parameterName=BevelSize}
  n78 = MaterialFunctionCall(Radius=Subtract(A=@n48, B=Multiply(A=@n77) {constB=0.01})) {materialFunction=&SphereGradient-2D}
  n79 = LinearInterpolate(
   A=@n68.RGB
   B=@n69.RGB
   Alpha=LinearInterpolate(A=Multiply(A=@n73, B=@n74), B=Multiply(A=@n76, B=@n74), Alpha=SmoothStep(Min=@n46, Max=@n47, Value=@n78.Result 0-1))
  )
  n80 = MaterialFunctionCall(Float2=MaterialFunctionCall(UVs=@n70, "Rotation Angle (0-1)"=ScalarParameter() {defaultValue=0.75, parameterName=Centerrot}) {materialFunction=&CustomRotator}.Rotated Values) {materialFunction=&BreakOutFloat2Components}
  n81 = ScalarParameter() {defaultValue=4, parameterName=CenterPow}
  n82 = LinearInterpolate(A=@n69.RGB, B=@n68.RGB, Alpha=Multiply(A=Power(Base=@n80.R, Exp=@n81), B=ScalarParameter() {defaultValue=1.1, parameterName=CenterMult}))
  n83 = Multiply(A=ScalarParameter() {defaultValue=13.7273, parameterName=CenterSize}) {constB=0.01}
  n84 = SmoothStep(
   Min=@n46
   Max=@n47
   Value=MaterialFunctionCall(Radius=Subtract(A=@n48, B=@n83)) {materialFunction=&SphereGradient-2D}.Result 0-1
  )
  LinearInterpolate(
   A=LinearInterpolate(A=Add(A=@n64, B=@n67), B=LinearInterpolate(A=@n79, B=@n82, Alpha=@n84), Alpha=@n49)
   B=VectorParameter() {parameterName=DotColor, defaultValue=<rgba 0.8333 0.8333 0.8333 1>}.RGB
   Alpha=SmoothStep(Min=Subtract(A=@n55, B=@n57), Max=Add(A=@n55, B=@n57), Value=@n54.Result 0-1)
  )
 ]
}
