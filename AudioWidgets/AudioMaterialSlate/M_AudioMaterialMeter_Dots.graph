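# Translucent UI material for an audio meter widget, drawn as a grid of glowing dots lit up to a VALUE parameter.
# Rotatable UVs feed MF_Dots for dot and glow masks, while a power-curved horizontal position drives a three-color gradient.
# VALUE is compared against position quantized to dot columns, so whole columns switch on.
# Glow strengthens along the meter, and unlit columns show flat DotsOffColor.
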Material /AudioWidgets/AudioMaterialSlate/M_AudioMaterialMeter_Dots
graph {
 properties {blendMode=BLEND_Translucent, materialDomain=MD_UI, pixelDepthOffsetMode=PDOM_Legacy, usedShadingModels=Default Lit}
 nodes [
  n1 = MaterialFunctionCall(UVs=TextureCoordinate(), "Rotation Angle (0-1)"=ScalarParameter() {parameterName="Rotation Angle (0-1) (S)"}) {materialFunction=&CustomRotator}
  n2 = MaterialFunctionCall(Float2=@n1.Rotated Values) {materialFunction=&BreakOutFloat2Components}
  n3 = ScalarParameter() {defaultValue=2.3027, parameterName=GradientPow}
  n4 = MaterialFunctionCall(
   A=VectorParameter() {parameterName="A (V3)", defaultValue=<rgba 0 0.1873 0.6016 0>}.RGB
   B=VectorParameter() {parameterName="B (V3)", defaultValue=<rgba 0.5286 0.2453 0 0>}.RGB
   C=VectorParameter() {parameterName="C (V3)", defaultValue=<rgba 2 0.0639 0 0>}.RGB
   Alpha=Clamp(None=Power(Base=@n2.R, Exp=@n3))
  ) {materialFunction=&Lerp_3Color}
  n5 = ScalarParameter() {defaultValue=32, parameterName="NumDotsX (S)"}
  n6 = Multiply(A=@n2.R, B=@n5)
  n7 = Constant() {r=1}
  n8 = If(A=Divide(A=Ceil(None=@n6), B=@n5), B=ScalarParameter() {defaultValue=1, parameterName=VALUE}, "A > B"=Constant(), "A == B"=@n7, "A < B"=@n7)
  n9 = MaterialFunctionCall!(
   UV=@n1.Rotated Values
   NumDotsX=ScalarParameter() {defaultValue=32, parameterName="NumDotsX (S)"}
   NumDotsY=ScalarParameter() {defaultValue=2, parameterName="NumDotsY (S)"}
   SharpRadius=ScalarParameter() {defaultValue=0.412, parameterName="SharpRadius (S)"}
   GlowDensity=ScalarParameter() {defaultValue=1.0235, parameterName="GlowDensity (S)"}
   GlowRadius=ScalarParameter() {defaultValue=0.6525, parameterName="GlowRadius (S)"}
   GlowMin=ScalarParameter() {parameterName="GlowMin (S)"}
   GlowMax=ScalarParameter() {defaultValue=1, parameterName="GlowMax (S)"}
  ) {materialFunction=/AudioWidgets/AudioMaterialSlate/Functions/MF_Dots}
  n10 = Multiply(A=Multiply(A=@n2.R, B=ScalarParameter() {defaultValue=2.5187, parameterName=GlowAmt}), B=@n8)
  n11 = LinearInterpolate(B=@n8, Alpha=Add(A=@n9.Dots, B=Multiply(A=@n9.Glow, B=@n10)))
  LinearInterpolate(
   A=VectorParameter() {parameterName=DotsOffColor, defaultValue=<rgba 0 0 0 1>}.RGB
   B=LinearInterpolate(B=Multiply(A=@n4.Result, B=@n11), Alpha=@n11)
   Alpha=@n8
  )
 ]
}
