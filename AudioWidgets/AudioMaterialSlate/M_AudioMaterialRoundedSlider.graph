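# Masked UI material that draws a vertical rounded slider for AudioWidgets, with an LED-style fill and a round thumb positioned by the VALUE parameter.
# GeneratedRoundRect gradients shape the bar, mask the glowing fill up to VALUE, and form an inner shadow by subtracting an offset copy.
# SphereGradient-2D calls, centered on a height interpolated from VALUE, add the thumb's offset drop shadow and beveled face, composited over the bar through chained lerps.
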Material /AudioWidgets/AudioMaterialSlate/M_AudioMaterialRoundedSlider
graph {
 properties {blendMode=BLEND_Masked, materialDomain=MD_UI, pixelDepthOffsetMode=PDOM_Legacy, usedShadingModels=Default Lit}
 nodes [
  n1 = VectorParameter() {parameterName=LedColor, defaultValue=<rgba 0 1.2949 2 1>}
  n2 = MaterialFunctionCall(UVs=TextureCoordinate(), Texture Scale=VectorParameter() {parameterName="Texture Scale (V2)", defaultValue=<rgba 1 1 1 0>}.RGB) {materialFunction=&ScaleUVsByCenter}
  n3 = MaterialFunctionCall!(
   UV Coords=@n2.UVs
   Box Dimensions=AppendVector(A=Constant() {r=0.03}, B=Constant() {r=0.7})
   CornerRadius=ScalarParameter() {defaultValue=0.0248, parameterName="CornerRadius (S)", sliderMax=0.05}
   Sharpness=Constant()
   Center
  ) {materialFunction=/Engine/Functions/Engine_MaterialFunctions01/Gradient/GeneratedRoundRect}
  n4 = ScalarParameter() {defaultValue=-0.3063, parameterName=LedInnerGlowPow}
  n5 = Multiply(A=@n1.RGB, B=Multiply(A=Power(Base=@n3.Result, Exp=@n4), B=ScalarParameter() {defaultValue=-0.6879, parameterName=CenterLedDarken}))
  n6 = Constant() {r=1}
  n7 = Constant()
  n8 = Constant() {r=1}
  n9 = ScalarParameter() {defaultValue=0.6378, parameterName=ScleY}
  n10 = MaterialFunctionCall(Float2=MaterialFunctionCall(UVs=@n2.UVs, Texture Scale=AppendVector(A=@n8, B=@n9)) {materialFunction=&ScaleUVsByCenter}.UVs) {materialFunction=&BreakOutFloat2Components}
  n11 = OneMinus(None=ScalarParameter() {defaultValue=0.672, parameterName=VALUE, sliderMax=1})
  n12 = Constant() {r=1}
  n13 = Constant()
  n14 = LinearInterpolate(
   A=VectorParameter() {parameterName=BarColor, defaultValue=<rgba 0.0156 0.0156 0.0156 1>}.RGB
   B=Multiply(A=Add(A=@n1.RGB, B=@n5), B=ScalarParameter() {defaultValue=1, parameterName=LedInt})
   Alpha=Multiply(A=If(A=@n3.Result, "A > B"=@n6, "A == B"=@n6, "A < B"=@n7) {constB=0.4}, B=If(A=@n10.G, B=@n11, "A > B"=@n12, "A == B"=@n12, "A < B"=@n13))
  )
  n15 = VectorParameter() {parameterName=BarInnerShadow, defaultValue=<rgba 0 0 0 1>}
  n16 = AppendVector(
   A=Constant() {r=0.5}
   B=LinearInterpolate(Alpha=ScalarParameter() {defaultValue=0.672, parameterName=VALUE, sliderMax=1}) {constA=0.83, constB=0.173}
  )
  n17 = AppendVector(A=Constant(), B=ScalarParameter() {defaultValue=0.007, parameterName=ShadowOffset})
  n18 = SmoothStep(Value=MaterialFunctionCall!(UVs=@n2.UVs, CenterPosition=Add(A=@n16, B=@n17), Radius=ScalarParameter() {defaultValue=0.03, parameterName=ShadowSize}) {materialFunction=/Engine/Functions/Engine_MaterialFunctions01/Gradient/SphereGradient-2D}.Result 0-1)
  n19 = ScalarParameter() {defaultValue=1.8822, parameterName=ShadowContrast}
  n20 = Clamp(None=Multiply(A=Power(Base=@n18, Exp=@n19), B=ScalarParameter() {defaultValue=1.6508, parameterName=ShadowIntensity}))
  n21 = MaterialFunctionCall(
   UV Coords=@n2.UVs
   Box Dimensions=AppendVector(A=Constant() {r=0.03}, B=Constant() {r=0.7})
   CornerRadius=ScalarParameter() {defaultValue=0.0248, parameterName="CornerRadius (S)", sliderMax=0.05}
   Sharpness=Constant() {r=100}
  ) {materialFunction=&GeneratedRoundRect}
  n22 = AppendVector(A=ScalarParameter() {parameterName=InnerShadowOffset_X}, B=ScalarParameter() {defaultValue=-0.008, parameterName=InnerShadowOffset_Y})
  n23 = Constant() {r=0.03}
  n24 = Constant() {r=0.7}
  n25 = Subtract(
   A=@n21.Result
   B=MaterialFunctionCall(
    UV Coords=Add(A=@n2.UVs, B=@n22)
    Box Dimensions=AppendVector(A=@n23, B=@n24)
    CornerRadius=ScalarParameter() {defaultValue=0.0248, parameterName="CornerRadius (S)", sliderMax=0.05}
    Sharpness=ScalarParameter() {defaultValue=1.7217, parameterName=InnershadoSharpen}
   ) {materialFunction=&GeneratedRoundRect}.Result
  )
  n26 = ScalarParameter() {defaultValue=0.146, parameterName=Min}
  n27 = ScalarParameter() {defaultValue=0.457, parameterName=Max}
  n28 = ScalarParameter() {defaultValue=0.672, parameterName=VALUE, sliderMax=1}
  n29 = MaterialFunctionCall(
   UVs=@n2.UVs
   CenterPosition=AppendVector(A=Constant() {r=0.5}, B=LinearInterpolate(Alpha=@n28) {constA=0.83, constB=0.173})
   Radius=ScalarParameter() {defaultValue=0.0262, parameterName=DotBevelRadius, sliderMax=0.03}
  ) {materialFunction=&SphereGradient-2D}
  n30 = AppendVector(
   A=Constant() {r=0.5}
   B=LinearInterpolate(Alpha=ScalarParameter() {defaultValue=0.672, parameterName=VALUE, sliderMax=1}) {constA=0.83, constB=0.173}
  )
  n31 = ScalarParameter() {defaultValue=0.19, parameterName="Radius (S)"}
  n32 = Constant() {r=1}
  LinearInterpolate(
   A=LinearInterpolate(A=LinearInterpolate(A=@n14, B=@n15.RGB, Alpha=@n20), B=@n15.RGB, Alpha=Clamp(None=@n25))
   B=LinearInterpolate(
    A=VectorParameter() {parameterName=ValueColorEdge, defaultValue=<rgba 0.0104 0.0104 0.0104 1>}.RGB
    B=VectorParameter() {parameterName=ValueColor, defaultValue=<rgba 0.0365 0.0365 0.0365 1>}.RGB
    Alpha=SmoothStep(Min=@n26, Max=@n27, Value=@n29.Result 0-1)
   )
   Alpha=If(
    A=MaterialFunctionCall(UVs=@n2.UVs, CenterPosition=@n30, Radius=@n31) {materialFunction=&SphereGradient-2D}.Result 0-1
    "A > B"=@n32
    "A == B"=@n32
    "A < B"=Constant()
   ) {constB=0.99}
  )
  Multiply(A=@n21.Result) {constB=0.4}
 ]
}
