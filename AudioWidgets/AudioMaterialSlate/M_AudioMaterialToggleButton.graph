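# Translucent UI material that draws a procedural round toggle button with an LED ring for AudioWidgets, without textures.
# Radial gradients and smoothsteps build the anti-aliased body, drop shadow and outer highlight, while rotated UV gradients shade the bevel and center.
# A ring line around the body blends from bar color to glowing LED color.
# The Click scalar drives button scale, shading, brightness and the angular-gradient sweep of the LED glow.
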Material /AudioWidgets/AudioMaterialSlate/M_AudioMaterialToggleButton
graph {
 properties {blendMode=BLEND_Translucent, materialDomain=MD_UI, pixelDepthOffsetMode=PDOM_Legacy, usedShadingModels=Default Lit}
 nodes [
  n1 = Constant() {r=0.1}
  n2 = ScalarParameter() {defaultValue=0.1, parameterName=AA}
  n3 = MaterialFunctionCall(
   UVs=TextureCoordinate()
   Texture Scale=LinearInterpolate(Alpha=ScalarParameter() {defaultValue=1, parameterName=Click, sliderMax=1}) {constA=1, constB=0.9}
  ) {materialFunction=&ScaleUVsByCenter}
  n4 = ScalarParameter() {defaultValue=0.28, parameterName="Radius (S)"}
  n5 = SmoothStep(
   Min=Subtract(A=@n1, B=@n2)
   Max=Add(A=@n1, B=@n2)
   Value=MaterialFunctionCall!(UVs=@n3.UVs, CenterPosition, Radius=@n4) {materialFunction=/Engine/Functions/Engine_MaterialFunctions01/Gradient/SphereGradient-2D}.Result 0-1
  )
  n6 = MaterialFunctionCall(
   UVs=TextureCoordinate()
   CenterPosition=AppendVector(A=ScalarParameter() {defaultValue=1.6808, parameterName=SmoothBevelX}, B=ScalarParameter() {defaultValue=0.5, parameterName=SmoothBevelY})
   Radius=ScalarParameter() {defaultValue=0.35, parameterName=SmoothBevelSize}
  ) {materialFunction=&SphereGradient-2D}
  n7 = Multiply(
   A=Power(Base=SmoothStep(Value=@n6.Result 0-1), Exp=ScalarParameter() {defaultValue=2, parameterName=SmoothBevelContrast})
   B=ScalarParameter() {defaultValue=0.7, parameterName=SmoothBevelInt}
  )
  n8 = MaterialFunctionCall(
   UVs=TextureCoordinate()
   CenterPosition=AppendVector(A=ScalarParameter() {defaultValue=0.5, parameterName=ShadowOffsetX}, B=ScalarParameter() {defaultValue=0.5, parameterName=ShadowOffsetY})
   Radius=ScalarParameter() {defaultValue=0.35, parameterName=ShadowSize}
  ) {materialFunction=&SphereGradient-2D}
  n9 = Multiply(
   A=Power(Base=SmoothStep(Value=@n8.Result 0-1), Exp=ScalarParameter() {defaultValue=2, parameterName=ShadowContrast})
   B=ScalarParameter() {defaultValue=0.7, parameterName=ShadowIntensity}
  )
  n10 = Add(A=@n7, B=@n9)
  n11 = ScalarParameter() {defaultValue=0.28, parameterName="Radius (S)"}
  n12 = ScalarParameter() {defaultValue=0.08, parameterName=CircleOffsetSize}
  n13 = MaterialFunctionCall!(
   LinePosition=ScalarParameter() {defaultValue=0.624, parameterName="LinePosition (S)"}
   LineThickness=ScalarParameter() {defaultValue=2, parameterName="LineThickness (S)"}
   Use V=StaticBoolParameter() {parameterName="Use V (B)"}
   LineSmoothStepThickness
   LineSmoothGlowThickness=ScalarParameter() {defaultValue=0.01, parameterName="LineSmoothStepThickness (S)"}
   UV=MaterialFunctionCall(Radius=Add(A=@n11, B=@n12)) {materialFunction=&SphereGradient-2D}.Result 0-1
  ) {materialFunction=/AudioWidgets/AudioMaterialSlate/Functions/MF_SingleLine}
  Clamp(None=Add(A=Add(A=@n5, B=@n10), B=@n13.Glow))
  n14 = Multiply(A=@n7, B=VectorParameter() {parameterName=SmoothBevelColor, defaultValue=<rgba 0.9427 0.9427 0.9427 1>}.RGB)
  n15 = Multiply(A=@n9, B=VectorParameter() {parameterName=ShadowColor, defaultValue=<rgba 0 0 0 1>}.RGB)
  n16 = VectorParameter() {group=Design, parameterName=Color_1, defaultValue=<rgba 0.95 0.95 0.95 1>}
  n17 = VectorParameter() {group=Design, parameterName=Color_2, defaultValue=<rgba 0.1266 0.1387 0.15 1>}
  n18 = MaterialFunctionCall(UVs=@n3.UVs, "Rotation Angle (0-1)"=ScalarParameter() {group=Gradient, parameterName=EdgeBevelRot}) {materialFunction=&CustomRotator}
  n19 = ScalarParameter() {defaultValue=0.2, group=Gradient, parameterName=Power}
  n20 = LinearInterpolate(Alpha=ScalarParameter() {defaultValue=1, parameterName=Click}) {constA=0.8, constB=0.04}
  n21 = Multiply(
   A=Power(
    Base=MaterialFunctionCall(Float2=@n18.Rotated Values) {materialFunction=&BreakOutFloat2Components}.R
    Exp=@n19
   )
   B=@n20
  )
  n22 = MaterialFunctionCall(UVs=@n3.UVs, "Rotation Angle (0-1)"=ScalarParameter() {defaultValue=0.5, parameterName=BevelRot}) {materialFunction=&CustomRotator}
  n23 = Multiply(
   A=Power(
    Base=MaterialFunctionCall(Float2=@n22.Rotated Values) {materialFunction=&BreakOutFloat2Components}.R
    Exp=@n19
   )
   B=@n20
  )
  n24 = Constant() {r=0.5}
  n25 = ScalarParameter() {defaultValue=0.392, parameterName=BevelSoftness}
  n26 = Subtract(A=@n24, B=@n25)
  n27 = Add(A=@n24, B=@n25)
  n28 = Multiply(A=ScalarParameter() {defaultValue=1.6724, group=Bevel, parameterName=BevelSize}) {constB=0.01}
  n29 = SmoothStep(
   Min=@n26
   Max=@n27
   Value=MaterialFunctionCall(UVs=@n3.UVs, Radius=Subtract(A=@n4, B=@n28)) {materialFunction=&SphereGradient-2D}.Result 0-1
  )
  n30 = Multiply(
   A=VectorParameter() {parameterName=MainColor, defaultValue=<rgba 1 1 1 1>}.RGB
   B=LinearInterpolate(Alpha=ScalarParameter() {defaultValue=1, parameterName=Click, sliderMax=1}) {constA=1, constB=2}
  )
  n31 = Multiply(A=LinearInterpolate(A=@n16.RGB, B=@n17.RGB, Alpha=LinearInterpolate(A=@n21, B=@n23, Alpha=@n29)), B=@n30)
  n32 = ScalarParameter() {defaultValue=0.75, parameterName=Centerrot}
  n33 = Power(
   Base=MaterialFunctionCall(Float2=MaterialFunctionCall(UVs=@n3.UVs, "Rotation Angle (0-1)"=@n32) {materialFunction=&CustomRotator}.Rotated Values) {materialFunction=&BreakOutFloat2Components}.R
   Exp=ScalarParameter() {defaultValue=4, parameterName=CenterPow}
  )
  n34 = ScalarParameter() {defaultValue=1.1, parameterName=CenterMult}
  n35 = Multiply(A=LinearInterpolate(A=@n17.RGB, B=@n16.RGB, Alpha=Multiply(A=@n33, B=@n34)), B=@n30)
  n36 = Multiply(A=ScalarParameter() {defaultValue=3.7153, parameterName=CenterSize}) {constB=0.01}
  n37 = SmoothStep(
   Min=@n26
   Max=@n27
   Value=MaterialFunctionCall(UVs=@n3.UVs, Radius=Subtract(A=@n4, B=@n36)) {materialFunction=&SphereGradient-2D}.Result 0-1
  )
  n38 = VectorParameter() {parameterName=LedColor, defaultValue=<rgba 0.0819 0.6202 1 1>}
  n39 = ScalarParameter() {defaultValue=1, parameterName=Click, sliderMax=1}
  n40 = MaterialFunctionCall!(
   LinePosition
   LineThickness=LinearInterpolate(B=ScalarParameter() {defaultValue=85, parameterName=AnimLineThickness}, Alpha=@n39)
   Use V
   UV=MaterialFunctionCall!(UV, Rotate=LinearInterpolate(Alpha=@n39) {constB=-1}) {materialFunction=/AudioWidgets/AudioMaterialSlate/Functions/MF_AngleGradient}.Result
  ) {materialFunction=/AudioWidgets/AudioMaterialSlate/Functions/MF_SingleLine_Smooth}
  n41 = Multiply(A=Multiply(A=@n40.Result, B=@n38.RGB), B=ScalarParameter() {defaultValue=1.4939, parameterName=AnimGlow})
  LinearInterpolate(
   A=LinearInterpolate(A=Subtract(A=@n14, B=@n15), B=LinearInterpolate(A=@n31, B=@n35, Alpha=@n37), Alpha=@n5)
   B=LinearInterpolate(
    A=VectorParameter() {parameterName=BarColor, defaultValue=<rgba 1 1 1 1>}.RGB
    B=Add(A=@n38.RGB, B=@n41)
    Alpha=Multiply(A=@n39, B=@n13.Glow)
   )
   Alpha=@n13.Glow
  )
  Multiply(A=@n40.Result, B=LinearInterpolate(Alpha=@n39) {constB=100})
 ]
}
